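# Physics asset for the F_MED_SwoopClasp_Pry female player body cosmetic, set up for secondary motion on a dangling ID card.
# A three-bone dyn_IDcard capsule chain has its master locked to spine_02, with swing-limited, nearly twist-free joints down to the tip.
# Pelvis, lower spine and left thigh capsules provide body volumes around the card.
# Spine_01 and thigh_l rotate freely against the pelvis, while spine_02 is rotation-locked to spine_01.
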PhysicsAsset /BRCosmetics/Characters/Player/Female/Medium/Bodies/F_MED_SwoopClasp_Pry/Meshes/F_MED_SwoopClasp_Pry_Physics
bodies {
 dyn_IDcard_01 [{shapeType=Capsule, length=1, radius=1, shapeName=dyn_IDcard_01_capsule, center=<1.3644 -0.0531 0.0622>, rotation=<rot 0 180 -90>}]
 dyn_IDcard_02 [{
   shapeType=Capsule
   length=5.2916
   radius=1.9277
   shapeName=dyn_IDcard_02_capsule
   center=<4.6923 -0.0541 1.0701>
   rotation=<rot 0 -91.0864 -88.9153>
  }]
 dyn_IDcard_MSTR [{shapeType=Capsule, length=1, radius=1, shapeName=dyn_IDcard_MSTR_capsule, rotation=<rot 0.0001 0 93.8549>}]
 pelvis [{shapeType=Capsule, length=8.5622, radius=9.044, shapeName=pelvis_capsule, center=<0 -1.5101 0>}]
 spine_01 [{shapeType=Capsule, length=7.7884, radius=6.6914, shapeName=spine_01_capsule, center=<0.4115 0 0>}]
 spine_02 [{shapeType=Capsule, length=9.4267, radius=4.3905, shapeName=spine_02_capsule, rotation=<rot 0 -90 0>}]
 thigh_l [{
   shapeType=Capsule
   length=20.595
   radius=7.4915
   shapeName=thigh_l_capsule
   center=<-15.073 0.0068 0.442>
   rotation=<rot 89.548 179.9737 179.9994>
  }]
}
constraints [
 {bone1Name=dyn_IDcard_02, bone2Name=dyn_IDcard_01, swing1LimitDegrees=23.4, swing2LimitDegrees=90, twistLimitDegrees=1}
 {bone1Name=dyn_IDcard_01, bone2Name=dyn_IDcard_MSTR, swing1LimitDegrees=10.44, swing2LimitDegrees=90, twistLimitDegrees=1}
 {
  bone1Name=dyn_IDcard_MSTR
  bone2Name=spine_02
  swing1LimitDegrees=45
  swing1Motion=Locked
  swing2LimitDegrees=45
  swing2Motion=Locked
  twistLimitDegrees=45
  twistMotion=Locked
 }
 {
  bone1Name=spine_01
  bone2Name=pelvis
  swing1LimitDegrees=45
  swing1Motion=Free
  swing2LimitDegrees=45
  swing2Motion=Free
  twistLimitDegrees=45
  twistMotion=Free
 }
 {
  bone1Name=thigh_l
  bone2Name=pelvis
  swing1LimitDegrees=45
  swing1Motion=Free
  swing2LimitDegrees=45
  swing2Motion=Free
  twistLimitDegrees=45
  twistMotion=Free
 }
 {
  bone1Name=spine_02
  bone2Name=spine_01
  swing1LimitDegrees=45
  swing1Motion=Locked
  swing2LimitDegrees=45
  swing2Motion=Locked
  twistLimitDegrees=45
  twistMotion=Locked
 }
]
