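# Post-copy animation Blueprint for the SwoopClasp_Pry female body cosmetic, adding secondary motion to its strap and pant legs.
# The parent mesh pose runs through Control Rig and RigidBody physics into a cached pose.
# That feeds two AnimDynamics chains with spherical calf limits on pants, selecting world-space simulation in the front end.
# The EventGraph drives material WPO cloth from foot velocity and calf transforms, enabling wind during skydive floating.
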AnimBlueprint /BRCosmetics/Characters/Player/Female/Medium/Bodies/F_MED_SwoopClasp_Pry/Meshes/F_MED_SwoopClasp_Pry_AnimBP
target_skeleton=&/F_MED_SwoopClasp_Pry_Skeleton
graphs [
 AnimGraph {nodes [
   n1 = Anim Dynamics - Bone: dyn_strap_01(ComponentPose=Local To Component(LocalPose=Use cached pose 'SavedPose20774'()), Alpha=1, ExternalForce="0,0,0") {
    angularDampingOverride=0.999
    angularSpringConstant=500
    bAngularSpring=true
    bOverrideAngularDamping=true
    bUseGravityOverride=true
    alphaBoolBlend {blendOption=Linear}
    boundBone {boneName=dyn_strap_01}
    componentAppliedLinearAccClamp=<1000 1000 1000>
    componentLinearAccScale=<0.2 0.2 0.2>
    componentLinearVelScale=<0.2 0.2 0.2>
    gravityOverride {z=-980}
    physicsBodyDefinitions [{
      boundBone {boneName=dyn_strap_01}
      boxExtents=<10 10 10>
      constraintSetup {angularLimitsMax=<0 0 0>, angularLimitsMin=<0 0 -10>, angularTarget=<0 0 0>}
      localJointOffset=<1 0 0>
     }]
   }
   n2 = Anim Dynamics - Bone: dyn_strap_01(ComponentPose=@n1, Alpha=1, ExternalForce="0,0,0") {
    angularDampingOverride=0.999
    angularSpringConstant=500
    bAngularSpring=true
    bOverrideAngularDamping=true
    bUseGravityOverride=true
    alphaBoolBlend {blendOption=Linear}
    boundBone {boneName=dyn_strap_01}
    componentAppliedLinearAccClamp=<1000 1000 1000>
    componentLinearAccScale=<0.2 0.2 0.2>
    componentLinearVelScale=<0.2 0.2 0.2>
    gravityOverride {z=-980}
    physicsBodyDefinitions [{
      boundBone {boneName=dyn_strap_01}
      boxExtents=<10 10 10>
      constraintSetup {angularLimitsMax=<0 0 3>, angularLimitsMin=<0 0 -20>, angularTarget=<0 0 0>}
      localJointOffset=<1 0 0>
     }]
   }
   n3 = Anim Dynamics - Bone: dyn_l_pant_MSTR(ComponentPose=@n2, Alpha=1, ExternalForce="0,0,0") {
    angularDampingOverride=0.999
    angularSpringConstant=500
    bAngularSpring=true
    bOverrideAngularDamping=true
    bUseGravityOverride=true
    bUsePlanarLimit=false
    bUseSphericalLimits=true
    alphaBoolBlend {blendOption=Linear}
    boundBone {boneName=dyn_l_pant_MSTR}
    componentAppliedLinearAccClamp=<2000 2000 2000>
    componentLinearAccScale=<0.1 0.1 0.1>
    componentLinearVelScale=<0.1 0.1 0.1>
    gravityOverride {z=-980}
    physicsBodyDefinitions [{
      boundBone {boneName=dyn_l_pant_MSTR}
      boxExtents=<10 10 10>
      constraintSetup {angularLimitsMax=<1 1 1>, angularLimitsMin=<-1 -1 -1>, angularTarget=<0 0 0>}
      localJointOffset=<-1 0 0>
     }]
    sphericalLimits [{limitRadius=6, drivingBone {boneName=calf_twist_01_l}, sphereLocalOffset=<9.1642 0 0>}]
   }
   n4 = Anim Dynamics - Bone: dyn_l_pant_outer(ComponentPose=@n3, Alpha=1, ExternalForce="0,0,0") {
    angularDampingOverride=0.999
    angularSpringConstant=500
    bAngularSpring=true
    bOverrideAngularDamping=true
    bUseGravityOverride=true
    bUsePlanarLimit=false
    bUseSphericalLimits=true
    alphaBoolBlend {blendOption=Linear}
    boundBone {boneName=dyn_l_pant_outer}
    componentAppliedLinearAccClamp=<2000 2000 2000>
    componentLinearAccScale=<0.5 0.5 0.5>
    componentLinearVelScale=<0.5 0.5 0.5>
    gravityOverride {z=-980}
    physicsBodyDefinitions [{
      boundBone {boneName=dyn_l_pant_outer}
      boxExtents=<10 10 10>
      constraintSetup {angularLimitsMax=<0 -1 1>, angularLimitsMin=<0 -5 -1>, angularTarget=<0 0 0>}
      localJointOffset=<4.5 0 0>
     }]
    sphericalLimits [{limitRadius=6, drivingBone {boneName=calf_twist_01_l}, sphereLocalOffset=<9.1642 0 0>}]
   }
   n5 = Anim Dynamics - Bone: dyn_l_pant_back(ComponentPose=@n4, Alpha=1, ExternalForce="0,0,0") {
    angularDampingOverride=0.999
    angularSpringConstant=500
    bAngularSpring=true
    bOverrideAngularDamping=true
    bUseGravityOverride=true
    bUsePlanarLimit=false
    bUseSphericalLimits=true
    alphaBoolBlend {blendOption=Linear}
    boundBone {boneName=dyn_l_pant_back}
    componentAppliedLinearAccClamp=<2000 2000 2000>
    componentLinearAccScale=<0.5 0.5 0.5>
    componentLinearVelScale=<0.5 0.5 0.5>
    gravityOverride {z=-980}
    physicsBodyDefinitions [{
      boundBone {boneName=dyn_l_pant_back}
      boxExtents=<10 10 10>
      constraintSetup {angularLimitsMax=<0 -1 1>, angularLimitsMin=<0 -5 -1>, angularTarget=<0 0 0>}
      localJointOffset=<4.5 0 0>
     }]
    sphericalLimits [{limitRadius=6, drivingBone {boneName=calf_twist_01_l}, sphereLocalOffset=<9.2206 0 0>}]
   }
   n6 = Anim Dynamics - Bone: dyn_l_pant_inner(ComponentPose=@n5, Alpha=1, ExternalForce="0,0,0") {
    angularDampingOverride=0.999
    angularSpringConstant=500
    bAngularSpring=true
    bOverrideAngularDamping=true
    bUseGravityOverride=true
    bUsePlanarLimit=false
    bUseSphericalLimits=true
    alphaBoolBlend {blendOption=Linear}
    boundBone {boneName=dyn_l_pant_inner}
    componentAppliedLinearAccClamp=<2000 2000 2000>
    componentLinearAccScale=<0.5 0.5 0.5>
    componentLinearVelScale=<0.5 0.5 0.5>
    gravityOverride {z=-980}
    physicsBodyDefinitions [{
      boundBone {boneName=dyn_l_pant_inner}
      boxExtents=<10 10 10>
      constraintSetup {angularLimitsMax=<0 -1 1>, angularLimitsMin=<0 -5 -1>, angularTarget=<0 0 0>}
      localJointOffset=<4.5 0 0>
     }]
    sphericalLimits [{limitRadius=6, drivingBone {boneName=calf_twist_01_l}, sphereLocalOffset=<9.5966 0 0>}]
   }
   n7 = Anim Dynamics - Bone: dyn_l_pant_front(ComponentPose=@n6, Alpha=1, ExternalForce="0,0,0") {
    angularDampingOverride=0.999
    angularSpringConstant=500
    bAngularSpring=true
    bOverrideAngularDamping=true
    bUseGravityOverride=true
    bUsePlanarLimit=false
    bUseSphericalLimits=true
    alphaBoolBlend {blendOption=Linear}
    boundBone {boneName=dyn_l_pant_front}
    componentAppliedLinearAccClamp=<2000 2000 2000>
    componentLinearAccScale=<0.5 0.5 0.5>
    componentLinearVelScale=<0.5 0.5 0.5>
    gravityOverride {z=-980}
    physicsBodyDefinitions [{
      boundBone {boneName=dyn_l_pant_front}
      boxExtents=<10 10 10>
      constraintSetup {angularLimitsMax=<0 -1 1>, angularLimitsMin=<0 -5 -1>, angularTarget=<0 0 0>}
      localJointOffset=<4.5 0 0>
     }]
    sphericalLimits [{limitRadius=6, drivingBone {boneName=calf_twist_01_l}, sphereLocalOffset=<9.214 0 0>}]
   }
   n8 = Anim Dynamics - Bone: dyn_r_pant_MSTR1(ComponentPose=@n7, Alpha=1, ExternalForce="0,0,0") {
    angularDampingOverride=0.999
    angularSpringConstant=500
    bAngularSpring=true
    bOverrideAngularDamping=true
    bUseGravityOverride=true
    bUsePlanarLimit=false
    bUseSphericalLimits=true
    alphaBoolBlend {blendOption=Linear}
    boundBone {boneName=dyn_r_pant_MSTR1}
    componentAppliedLinearAccClamp=<2000 2000 2000>
    componentLinearAccScale=<0.1 0.1 0.1>
    componentLinearVelScale=<0.1 0.1 0.1>
    gravityOverride {z=-980}
    physicsBodyDefinitions [{
      boundBone {boneName=dyn_r_pant_MSTR1}
      boxExtents=<10 10 10>
      constraintSetup {angularLimitsMax=<1 1 1>, angularLimitsMin=<-1 -1 -1>, angularTarget=<0 0 0>}
      localJointOffset=<1 0 0>
     }]
    sphericalLimits [{limitRadius=6, drivingBone {boneName=calf_twist_01_r}, sphereLocalOffset=<-9 0 0>}]
   }
   n9 = Anim Dynamics - Bone: dyn_r_pant_outer(ComponentPose=@n8, Alpha=1, ExternalForce="0,0,0") {
    angularDampingOverride=0.999
    angularSpringConstant=500
    bAngularSpring=true
    bOverrideAngularDamping=true
    bUseGravityOverride=true
    bUsePlanarLimit=false
    bUseSphericalLimits=true
    alphaBoolBlend {blendOption=Linear}
    boundBone {boneName=dyn_r_pant_outer}
    componentAppliedLinearAccClamp=<2000 2000 2000>
    componentLinearAccScale=<0.5 0.5 0.5>
    componentLinearVelScale=<0.5 0.5 0.5>
    gravityOverride {z=-980}
    physicsBodyDefinitions [{
      boundBone {boneName=dyn_r_pant_outer}
      boxExtents=<10 10 10>
      constraintSetup {angularLimitsMax=<0 -1 1>, angularLimitsMin=<0 -5 -1>, angularTarget=<0 0 0>}
      localJointOffset=<4.5 0 0>
     }]
    sphericalLimits [{limitRadius=6, drivingBone {boneName=calf_twist_01_r}, sphereLocalOffset=<-9 0 0>}]
   }
   n10 = Anim Dynamics - Bone: dyn_r_pant_back(ComponentPose=@n9, Alpha=1, ExternalForce="0,0,0") {
    angularDampingOverride=0.999
    angularSpringConstant=500
    bAngularSpring=true
    bOverrideAngularDamping=true
    bUseGravityOverride=true
    bUsePlanarLimit=false
    bUseSphericalLimits=true
    alphaBoolBlend {blendOption=Linear}
    boundBone {boneName=dyn_r_pant_back}
    componentAppliedLinearAccClamp=<2000 2000 2000>
    componentLinearAccScale=<0.5 0.5 0.5>
    componentLinearVelScale=<0.5 0.5 0.5>
    gravityOverride {z=-980}
    physicsBodyDefinitions [{
      boundBone {boneName=dyn_r_pant_back}
      boxExtents=<10 10 10>
      constraintSetup {angularLimitsMax=<0 -1 1>, angularLimitsMin=<0 -5 -1>, angularTarget=<0 0 0>}
      localJointOffset=<4.5 0 0>
     }]
    sphericalLimits [{limitRadius=6, drivingBone {boneName=calf_twist_01_r}, sphereLocalOffset=<-9 0 0>}]
   }
   n11 = Anim Dynamics - Bone: dyn_r_pant_inner(ComponentPose=@n10, Alpha=1, ExternalForce="0,0,0") {
    angularDampingOverride=0.999
    angularSpringConstant=500
    bAngularSpring=true
    bOverrideAngularDamping=true
    bUseGravityOverride=true
    bUsePlanarLimit=false
    bUseSphericalLimits=true
    alphaBoolBlend {blendOption=Linear}
    boundBone {boneName=dyn_r_pant_inner}
    componentAppliedLinearAccClamp=<2000 2000 2000>
    componentLinearAccScale=<0.5 0.5 0.5>
    componentLinearVelScale=<0.5 0.5 0.5>
    gravityOverride {z=-980}
    physicsBodyDefinitions [{
      boundBone {boneName=dyn_r_pant_inner}
      boxExtents=<10 10 10>
      constraintSetup {angularLimitsMax=<0 -1 1>, angularLimitsMin=<0 -5 -1>, angularTarget=<0 0 0>}
      localJointOffset=<4.5 0 0>
     }]
    sphericalLimits [{limitRadius=6, drivingBone {boneName=calf_twist_01_r}, sphereLocalOffset=<-9 0 0>}]
   }
   n12 = Anim Dynamics - Bone: dyn_r_pant_front(ComponentPose=@n11, Alpha=1, ExternalForce="0,0,0") {
    angularDampingOverride=0.999
    angularSpringConstant=500
    bAngularSpring=true
    bOverrideAngularDamping=true
    bUseGravityOverride=true
    bUsePlanarLimit=false
    bUseSphericalLimits=true
    alphaBoolBlend {blendOption=Linear}
    boundBone {boneName=dyn_r_pant_front}
    componentAppliedLinearAccClamp=<2000 2000 2000>
    componentLinearAccScale=<0.5 0.5 0.5>
    componentLinearVelScale=<0.5 0.5 0.5>
    gravityOverride {z=-980}
    physicsBodyDefinitions [{
      boundBone {boneName=dyn_r_pant_front}
      boxExtents=<10 10 10>
      constraintSetup {angularLimitsMax=<0 -1 1>, angularLimitsMin=<0 -5 -1>, angularTarget=<0 0 0>}
      localJointOffset=<4.5 0 0>
     }]
    sphericalLimits [{limitRadius=6, drivingBone {boneName=calf_twist_01_r}, sphereLocalOffset=<-9 0 0>}]
   }
   n13 = Anim Dynamics - Bone: dyn_strap_01(ComponentPose=Local To Component(LocalPose=Use cached pose 'SavedPose20774'()), Alpha=1, ExternalForce="0,0,0") {
    angularBiasOverride=0.999
    angularDampingOverride=0.999
    angularSpringConstant=500
    bAngularSpring=true
    bOverrideAngularBias=true
    bOverrideAngularDamping=true
    bUseGravityOverride=true
    simulationSpace=World
    alphaBoolBlend {blendOption=Linear}
    boundBone {boneName=dyn_strap_01}
    componentAppliedLinearAccClamp=<1000 1000 1000>
    componentLinearAccScale=<0.2 0.2 0.2>
    componentLinearVelScale=<0.2 0.2 0.2>
    gravityOverride {z=-980}
    physicsBodyDefinitions [{
      boundBone {boneName=dyn_strap_01}
      boxExtents=<10 1 10>
      constraintSetup {angularLimitsMax=<0 0 0>, angularLimitsMin=<0 0 -10>, angularTarget=<0 0 0>}
      localJointOffset=<1 0 0>
     }]
   }
   n14 = Anim Dynamics - Bone: dyn_strap_01(ComponentPose=@n13, Alpha=1, ExternalForce="0,0,0") {
    angularBiasOverride=0.999
    angularDampingOverride=0.999
    angularSpringConstant=500
    bAngularSpring=true
    bOverrideAngularBias=true
    bOverrideAngularDamping=true
    bUseGravityOverride=true
    simulationSpace=World
    alphaBoolBlend {blendOption=Linear}
    boundBone {boneName=dyn_strap_01}
    componentAppliedLinearAccClamp=<1000 1000 1000>
    componentLinearAccScale=<0.2 0.2 0.2>
    componentLinearVelScale=<0.2 0.2 0.2>
    gravityOverride {z=-980}
    physicsBodyDefinitions [{
      boundBone {boneName=dyn_strap_01}
      boxExtents=<10 1 10>
      constraintSetup {angularLimitsMax=<0 0 3>, angularLimitsMin=<0 0 -20>, angularTarget=<0 0 0>}
      localJointOffset=<1 0 0>
     }]
   }
   n15 = Anim Dynamics - Bone: dyn_l_pant_MSTR(ComponentPose=@n14, Alpha=1, ExternalForce="0,0,0") {
    angularBiasOverride=0.999
    angularDampingOverride=0.999
    angularSpringConstant=500
    bAngularSpring=true
    bOverrideAngularBias=true
    bOverrideAngularDamping=true
    bUseGravityOverride=true
    bUsePlanarLimit=false
    bUseSphericalLimits=true
    simulationSpace=World
    alphaBoolBlend {blendOption=Linear}
    boundBone {boneName=dyn_l_pant_MSTR}
    componentAppliedLinearAccClamp=<2000 2000 2000>
    componentLinearAccScale=<0.1 0.1 0.1>
    componentLinearVelScale=<0.1 0.1 0.1>
    gravityOverride {z=-980}
    physicsBodyDefinitions [{
      boundBone {boneName=dyn_l_pant_MSTR}
      boxExtents=<15 5 15>
      constraintSetup {angularLimitsMax=<1 1 1>, angularLimitsMin=<-1 -1 -1>, angularTarget=<0 0 0>}
      localJointOffset=<-1 0 0>
     }]
    sphericalLimits [{limitRadius=6, drivingBone {boneName=calf_twist_01_l}, sphereLocalOffset=<9.1642 0 0>}]
   }
   n16 = Anim Dynamics - Bone: dyn_l_pant_outer(ComponentPose=@n15, Alpha=1, ExternalForce="0,0,0") {
    angularBiasOverride=0.999
    angularDampingOverride=0.999
    angularSpringConstant=500
    bAngularSpring=true
    bOverrideAngularBias=true
    bOverrideAngularDamping=true
    bUseGravityOverride=true
    bUsePlanarLimit=false
    bUseSphericalLimits=true
    simulationSpace=World
    alphaBoolBlend {blendOption=Linear}
    boundBone {boneName=dyn_l_pant_outer}
    componentAppliedLinearAccClamp=<2000 2000 2000>
    componentLinearAccScale=<0.1 0.1 0.1>
    componentLinearVelScale=<0.1 0.1 0.1>
    gravityOverride {z=-980}
    physicsBodyDefinitions [{
      boundBone {boneName=dyn_l_pant_outer}
      boxExtents=<15 5 15>
      constraintSetup {angularLimitsMax=<0 -1 1>, angularLimitsMin=<0 -5 -1>, angularTarget=<0 0 0>}
      localJointOffset=<4.5 0 0>
     }]
    sphericalLimits [{limitRadius=6, drivingBone {boneName=calf_twist_01_l}, sphereLocalOffset=<9.1642 0 0>}]
   }
   n17 = Anim Dynamics - Bone: dyn_l_pant_back(ComponentPose=@n16, Alpha=1, ExternalForce="0,0,0") {
    angularBiasOverride=0.999
    angularDampingOverride=0.999
    angularSpringConstant=500
    bAngularSpring=true
    bOverrideAngularBias=true
    bOverrideAngularDamping=true
    bUseGravityOverride=true
    bUsePlanarLimit=false
    bUseSphericalLimits=true
    simulationSpace=World
    alphaBoolBlend {blendOption=Linear}
    boundBone {boneName=dyn_l_pant_back}
    componentAppliedLinearAccClamp=<2000 2000 2000>
    componentLinearAccScale=<0.1 0.1 0.1>
    componentLinearVelScale=<0.1 0.1 0.1>
    gravityOverride {z=-980}
    physicsBodyDefinitions [{
      boundBone {boneName=dyn_l_pant_back}
      boxExtents=<15 5 15>
      constraintSetup {angularLimitsMax=<0 -1 1>, angularLimitsMin=<0 -5 -1>, angularTarget=<0 0 0>}
      localJointOffset=<4.5 0 0>
     }]
    sphericalLimits [{limitRadius=6, drivingBone {boneName=calf_twist_01_l}, sphereLocalOffset=<9.2206 0 0>}]
   }
   n18 = Anim Dynamics - Bone: dyn_l_pant_inner(ComponentPose=@n17, Alpha=1, ExternalForce="0,0,0") {
    angularBiasOverride=0.999
    angularDampingOverride=0.999
    angularSpringConstant=500
    bAngularSpring=true
    bOverrideAngularBias=true
    bOverrideAngularDamping=true
    bUseGravityOverride=true
    bUsePlanarLimit=false
    bUseSphericalLimits=true
    simulationSpace=World
    alphaBoolBlend {blendOption=Linear}
    boundBone {boneName=dyn_l_pant_inner}
    componentAppliedLinearAccClamp=<2000 2000 2000>
    componentLinearAccScale=<0.1 0.1 0.1>
    componentLinearVelScale=<0.1 0.1 0.1>
    gravityOverride {z=-980}
    physicsBodyDefinitions [{
      boundBone {boneName=dyn_l_pant_inner}
      boxExtents=<15 5 15>
      constraintSetup {angularLimitsMax=<0 -1 1>, angularLimitsMin=<0 -5 -1>, angularTarget=<0 0 0>}
      localJointOffset=<4.5 0 0>
     }]
    sphericalLimits [{limitRadius=6, drivingBone {boneName=calf_twist_01_l}, sphereLocalOffset=<9.5966 0 0>}]
   }
   n19 = Anim Dynamics - Bone: dyn_l_pant_front(ComponentPose=@n18, Alpha=1, ExternalForce="0,0,0") {
    angularBiasOverride=0.999
    angularDampingOverride=0.999
    angularSpringConstant=500
    bAngularSpring=true
    bOverrideAngularBias=true
    bOverrideAngularDamping=true
    bUseGravityOverride=true
    bUsePlanarLimit=false
    bUseSphericalLimits=true
    simulationSpace=World
    alphaBoolBlend {blendOption=Linear}
    boundBone {boneName=dyn_l_pant_front}
    componentAppliedLinearAccClamp=<2000 2000 2000>
    componentLinearAccScale=<0.1 0.1 0.1>
    componentLinearVelScale=<0.1 0.1 0.1>
    gravityOverride {z=-980}
    physicsBodyDefinitions [{
      boundBone {boneName=dyn_l_pant_front}
      boxExtents=<15 5 15>
      constraintSetup {angularLimitsMax=<0 -1 1>, angularLimitsMin=<0 -5 -1>, angularTarget=<0 0 0>}
      localJointOffset=<4.5 0 0>
     }]
    sphericalLimits [{limitRadius=6, drivingBone {boneName=calf_twist_01_l}, sphereLocalOffset=<9.214 0 0>}]
   }
   n20 = Anim Dynamics - Bone: dyn_r_pant_MSTR1(ComponentPose=@n19, Alpha=1, ExternalForce="0,0,0") {
    angularBiasOverride=0.999
    angularDampingOverride=0.999
    angularSpringConstant=500
    bAngularSpring=true
    bOverrideAngularBias=true
    bOverrideAngularDamping=true
    bUseGravityOverride=true
    bUsePlanarLimit=false
    bUseSphericalLimits=true
    simulationSpace=World
    alphaBoolBlend {blendOption=Linear}
    boundBone {boneName=dyn_r_pant_MSTR1}
    componentAppliedLinearAccClamp=<2000 2000 2000>
    componentLinearAccScale=<0.1 0.1 0.1>
    componentLinearVelScale=<0.1 0.1 0.1>
    gravityOverride {z=-980}
    physicsBodyDefinitions [{
      boundBone {boneName=dyn_r_pant_MSTR1}
      boxExtents=<15 5 15>
      constraintSetup {angularLimitsMax=<1 1 1>, angularLimitsMin=<-1 -1 -1>, angularTarget=<0 0 0>}
      localJointOffset=<1 0 0>
     }]
    sphericalLimits [{limitRadius=6, drivingBone {boneName=calf_twist_01_r}, sphereLocalOffset=<-9 0 0>}]
   }
   n21 = Anim Dynamics - Bone: dyn_r_pant_outer(ComponentPose=@n20, Alpha=1, ExternalForce="0,0,0") {
    angularBiasOverride=0.999
    angularDampingOverride=0.999
    angularSpringConstant=500
    bAngularSpring=true
    bOverrideAngularBias=true
    bOverrideAngularDamping=true
    bUseGravityOverride=true
    bUsePlanarLimit=false
    bUseSphericalLimits=true
    simulationSpace=World
    alphaBoolBlend {blendOption=Linear}
    boundBone {boneName=dyn_r_pant_outer}
    componentAppliedLinearAccClamp=<2000 2000 2000>
    componentLinearAccScale=<0.1 0.1 0.1>
    componentLinearVelScale=<0.1 0.1 0.1>
    gravityOverride {z=-980}
    physicsBodyDefinitions [{
      boundBone {boneName=dyn_r_pant_outer}
      boxExtents=<15 5 15>
      constraintSetup {angularLimitsMax=<0 -1 1>, angularLimitsMin=<0 -5 -1>, angularTarget=<0 0 0>}
      localJointOffset=<4.5 0 0>
     }]
    sphericalLimits [{limitRadius=6, drivingBone {boneName=calf_twist_01_r}, sphereLocalOffset=<-9 0 0>}]
   }
   n22 = Anim Dynamics - Bone: dyn_r_pant_back(ComponentPose=@n21, Alpha=1, ExternalForce="0,0,0") {
    angularBiasOverride=0.999
    angularDampingOverride=0.999
    angularSpringConstant=500
    bAngularSpring=true
    bOverrideAngularBias=true
    bOverrideAngularDamping=true
    bUseGravityOverride=true
    bUsePlanarLimit=false
    bUseSphericalLimits=true
    simulationSpace=World
    alphaBoolBlend {blendOption=Linear}
    boundBone {boneName=dyn_r_pant_back}
    componentAppliedLinearAccClamp=<2000 2000 2000>
    componentLinearAccScale=<0.1 0.1 0.1>
    componentLinearVelScale=<0.1 0.1 0.1>
    gravityOverride {z=-980}
    physicsBodyDefinitions [{
      boundBone {boneName=dyn_r_pant_back}
      boxExtents=<15 5 15>
      constraintSetup {angularLimitsMax=<0 -1 1>, angularLimitsMin=<0 -5 -1>, angularTarget=<0 0 0>}
      localJointOffset=<4.5 0 0>
     }]
    sphericalLimits [{limitRadius=6, drivingBone {boneName=calf_twist_01_r}, sphereLocalOffset=<-9 0 0>}]
   }
   n23 = Anim Dynamics - Bone: dyn_r_pant_inner(ComponentPose=@n22, Alpha=1, ExternalForce="0,0,0") {
    angularBiasOverride=0.999
    angularDampingOverride=0.999
    angularSpringConstant=500
    bAngularSpring=true
    bOverrideAngularBias=true
    bOverrideAngularDamping=true
    bUseGravityOverride=true
    bUsePlanarLimit=false
    bUseSphericalLimits=true
    simulationSpace=World
    alphaBoolBlend {blendOption=Linear}
    boundBone {boneName=dyn_r_pant_inner}
    componentAppliedLinearAccClamp=<2000 2000 2000>
    componentLinearAccScale=<0.1 0.1 0.1>
    componentLinearVelScale=<0.1 0.1 0.1>
    gravityOverride {z=-980}
    physicsBodyDefinitions [{
      boundBone {boneName=dyn_r_pant_inner}
      boxExtents=<15 5 15>
      constraintSetup {angularLimitsMax=<0 -1 1>, angularLimitsMin=<0 -5 -1>, angularTarget=<0 0 0>}
      localJointOffset=<4.5 0 0>
     }]
    sphericalLimits [{limitRadius=6, drivingBone {boneName=calf_twist_01_r}, sphereLocalOffset=<-9 0 0>}]
   }
   n24 = Anim Dynamics - Bone: dyn_r_pant_front(ComponentPose=@n23, Alpha=1, ExternalForce="0,0,0") {
    angularBiasOverride=0.999
    angularDampingOverride=0.999
    angularSpringConstant=500
    bAngularSpring=true
    bOverrideAngularBias=true
    bOverrideAngularDamping=true
    bUseGravityOverride=true
    bUsePlanarLimit=false
    bUseSphericalLimits=true
    simulationSpace=World
    alphaBoolBlend {blendOption=Linear}
    boundBone {boneName=dyn_r_pant_front}
    componentAppliedLinearAccClamp=<2000 2000 2000>
    componentLinearAccScale=<0.1 0.1 0.1>
    componentLinearVelScale=<0.1 0.1 0.1>
    gravityOverride {z=-980}
    physicsBodyDefinitions [{
      boundBone {boneName=dyn_r_pant_front}
      boxExtents=<15 5 15>
      constraintSetup {angularLimitsMax=<0 -1 1>, angularLimitsMin=<0 -5 -1>, angularTarget=<0 0 0>}
      localJointOffset=<4.5 0 0>
     }]
    sphericalLimits [{limitRadius=6, drivingBone {boneName=calf_twist_01_r}, sphereLocalOffset=<-9 0 0>}]
   }
   Output Pose(Result=Blend Poses by bool(BlendPose_0=Component To Local(ComponentPose=@n24), BlendPose_1=Component To Local(ComponentPose=@n12), bActiveValue=Get bIsInFrontEnd()) {blendType=Linear}) {layerGroup=None}
   n25 = Control Rig(Source=Copy Pose From Mesh() {bUseAttachedParent=true}) {
    alphaBoolBlend {blendOption=Linear}
    controlRigAssetReference {blueprintRigClass=&/F_MED_SwoopClasp_Pry_CtrlRig.F_MED_SwoopClasp_Pry_CtrlRig_C}
   }
   n26 = RigidBody(ComponentPose=Local To Component(LocalPose=@n25)) {
    bClampLinearTranslationLimitToRefPose=true
    bOverrideWorldGravity=true
    overridePhysicsAsset=&/F_MED_SwoopClasp_Pry_Physics
    simulationSpace=BaseBoneSpace
    alphaBoolBlend {blendOption=Linear}
    baseBoneRef {boneName=pelvis}
    overrideWorldGravity {z=-980}
    simSpaceSettings {
     dampingAlpha=0
     maxAngularAcceleration=5
     maxAngularVelocity=5
     maxLinearAcceleration=5
     maxLinearVelocity=700
     velocityScaleZ=10
     worldAlpha=0.5
     externalLinearDragV=<5 5 10>
    }
   }
   Save cached pose 'SavedPose20774'(Pose=Component To Local(ComponentPose=@n26)) {cacheName=SavedPose20774}
  ]}
 EventGraph {nodes [
   n1 = Event BlueprintUpdateAnimation()
   > n2 = Set Delta Time X(Delta Time X=@n1.DeltaTimeX)
   > Is Valid(InputObject=Get PlayerPawn() -> (PlayerPawn:Pawn Object Reference))
   > SetWPOGlobalsFunction!(
    execute
    InMaterial:Material Instance Dynamic Object Reference=Get Mat Instance Dynamic()
    InOverlayMaterial:Material Instance Dynamic Object Reference=Get Overlay Mat Dynamic() -> (Overlay Mat Dynamic:Material Instance Dynamic Object Reference)
    InSkeletalMesh:Skeletal Mesh Component Object Reference=Get Skel Mesh()
    InLocalNoiseReferenceTransformSocketName:Name=None
    InLocalNoiseOffsetMultiplier:float=0
    InOutLocalNoiseTransformLocation:"Vector (by ref)"=Get Local Noise Transform Location() -> (Local Noise Transform Location:Vector)
    InOutLocalNoiseOffset:"Vector (by ref)"=Get Local Noise Offset()
    InLocalTurbulenceVectorTransformSocketName:Name=None
    InLocalTurbulenceVectorTransformAxis:EAxis Enum=None
   ) -> (then)
   n3 = Get Velocity Interpolation Speed() -> (Velocity Interpolation Speed:float)
   > SetWPORegionalFunction!(
    execute
    InMaterial:Material Instance Dynamic Object Reference=Get Mat Instance Dynamic()
    InOverlayMaterial:Material Instance Dynamic Object Reference=Get Overlay Mat Dynamic()
    InSkeletalMesh:Skeletal Mesh Component Object Reference=Get Skel Mesh()
    InVelocityTransformName:Name=foot_l
    InVelocityInterpolationSpeed:float=@n3
    InCollisionTransformSocketName:Name=calf_l
    InCollisionTransformAxis:EAxis Enum
    InShouldInvertCollisionTransformAxis:bool=false
    InDeltaTimeX:float=@n2
    InOutPreviousVelocityTransformLocation:"Vector (by ref)"=Get R1 Velocity Transform Location()
    InOutPreviousVelocityTransformVelocity:"Vector (by ref)"=Get R1 Velocity Transform Velocity()
    InRegion:int=1
   ) -> (then)
   > SetWPORegionalFunction(
    InMaterial=Get Mat Instance Dynamic()
    InOverlayMaterial=Get Overlay Mat Dynamic()
    InSkeletalMesh=Get Skel Mesh()
    InVelocityTransformName=foot_r
    InVelocityInterpolationSpeed=@n3
    InCollisionTransformSocketName=calf_r
    InShouldInvertCollisionTransformAxis=true
    InDeltaTimeX=@n2
    InOutPreviousVelocityTransformLocation=Get R2 Velocity Transform Location()
    InOutPreviousVelocityTransformVelocity=Get R2 Velocity Transform Velocity()
    InRegion=2
   )
   > SetScalarParameterValue(self=Get Mat Instance Dynamic(), ParameterName=ClothSim_WindStrength, Value=SelectFloat(A=1, B=0, bPickA=Get bIsSkydiveFloating()))
   > n4 = Branch(Condition=Get bIsSkydiveFloating())
   > n5 = Set WPO wind intensity(WPO wind intensity:double=1) -> (Output_Get:double)
   SetScalarParameterValue(
    execute=[@n5.then, @n6.then]
    self=[Get Mat Instance Dynamic(), Get Overlay Mat Dynamic()]
    ParameterName=ClothSim_WindStrength
    Value=Get WPO wind intensity() -> (WPO wind intensity:double)
   )
   n6 = Set WPO wind intensity(execute=@n4.else, WPO wind intensity=0)
   Event OnCharacterCustomizationLoaded()
   n7 = GetPartSkeletalMeshComponent()
   > n8 = Get Dynamic Material Instance(SkeletalMeshComponent=@n7)
   > Set Mat Instance Dynamic(Mat Instance Dynamic=@n8)
   > n9 = GetOrCreateDynamicOverlayMaterial(MeshComponent=@n7, MaterialName=None)
   > Set Overlay Mat Dynamic(Overlay Mat Dynamic:Material Instance Dynamic Object Reference=@n9) -> (Output_Get:Material Instance Dynamic Object Reference)
   Event BlueprintInitializeAnimation()
   > Set Skel Mesh(Skel Mesh=GetOwnerSkeletalMeshComponent())
   n10 = TryGetPawnOwner()
   > Is Valid(InputObject=@n10)
   > Set PlayerPawn(PlayerPawn:Pawn Object Reference=@n10) -> (Output_Get:Pawn Object Reference)
  ]}
]
variables [
 Delta Time X {type=number}
 Local Noise Offset {type=Vector}
 Local Noise Transform Location {type=Vector}
 Mat Instance Dynamic {type=MaterialInstanceDynamic}
 Overlay Mat Dynamic {type=MaterialInstanceDynamic}
 PlayerPawn {type=Pawn, default=null}
 R1 Velocity Transform Location {type=Vector}
 R1 Velocity Transform Velocity {type=Vector}
 R2 Velocity Transform Location {type=Vector}
 R2 Velocity Transform Velocity {type=Vector}
 Skel Mesh {type=SkeletalMeshComponent}
 Velocity Interpolation Speed {type=number}
 WPO wind intensity {type=number}
]
