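# Physics asset for the hair on the F_MED_SwoopClasp_Pry female character's face accessory mesh, defining capsule bodies for dynamic hair bones.
# Back, left and right chains each link to a MSTR root with limited swing and tight twist, with the back swinging widest.
# The right chain branches, with dyn_hair_r_03 constrained to r_00 under tighter swing limits.
# Unconstrained capsules on upper spine, neck, clavicles and upper arms provide body shapes around the hair.
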PhysicsAsset /BRCosmetics/Characters/Player/Female/Medium/Bodies/F_MED_SwoopClasp_Pry/Meshes/Parts/F_MED_SwoopClasp_Pry_FaceAcc_Physics_FE_Back_v2
bodies {
 clavicle_l [{shapeType=Capsule, length=8.5, radius=6, shapeName=clavicle_l_capsule, center=<1 0 -1>, rotation=<rot 90 0 0>}]
 clavicle_r [{shapeType=Capsule, length=8.5, radius=6, shapeName=clavicle_r_capsule, center=<-1 0 1>, rotation=<rot 90 0 0>}]
 dyn_hair_bk_00 [{shapeType=Capsule, length=4.5, radius=4.5, shapeName=dyn_hair_bk_00_capsule, center=<4 -1.5 0>}]
 dyn_hair_bk_01 [{shapeType=Capsule, length=5.5, radius=5.5, shapeName=dyn_hair_bk_01_capsule, center=<3.5 -2 0>}]
 dyn_hair_bk_02 [{shapeType=Capsule, length=4.5, radius=7, shapeName=dyn_hair_bk_02_capsule, center=<3 -4 0>}]
 dyn_hair_bk_03 [{shapeType=Capsule, length=5, radius=7, shapeName=dyn_hair_bk_03_capsule, center=<7 -4 0>}]
 dyn_hair_bk_MSTR [{shapeType=Capsule, length=2, radius=1, shapeName=dyn_hair_bk_MSTR_capsule}]
 dyn_hair_l_00 [{shapeType=Capsule, length=3.5, radius=4.5, shapeName=dyn_hair_l_00_capsule, center=<3 -2 0>}]
 dyn_hair_l_01 [{shapeType=Capsule, length=2.5, radius=5.5, shapeName=dyn_hair_l_01_capsule, center=<3.5 -2 0>}]
 dyn_hair_l_02 [{shapeType=Capsule, length=2.5, radius=6, shapeName=dyn_hair_l_02_capsule, center=<3 -4 0>}]
 dyn_hair_l_MSTR [{shapeType=Capsule, length=2, radius=1, shapeName=dyn_hair_l_MSTR_capsule}]
 dyn_hair_r_00 [{shapeType=Capsule, length=3, radius=4.5, shapeName=dyn_hair_r_00_capsule, center=<3 -2 0>}]
 dyn_hair_r_01 [{shapeType=Capsule, length=3, radius=5, shapeName=dyn_hair_r_01_capsule, center=<3.5 -2 0>}]
 dyn_hair_r_02 [{shapeType=Capsule, length=3, radius=6, shapeName=dyn_hair_r_02_capsule, center=<3 -4.4712 0>}]
 dyn_hair_r_03 [{
   shapeType=Capsule
   length=3.5
   radius=2
   shapeName=dyn_hair_r_03_capsule
   center=<2.9829 0.1178 -0.7244>
   rotation=<rot 76.954 13.1078 -0.0001>
  }]
 dyn_hair_r_MSTR [{shapeType=Capsule, length=2, radius=1, shapeName=dyn_hair_r_MSTR_capsule}]
 neck_01 [{shapeType=Capsule, length=6.005, radius=3.505, shapeName=neck_01_capsule, center=<1.8637 0 0>, rotation=<rot 90 0 0>}]
 spine_04 [{shapeType=Capsule, length=8.5, radius=7.5, shapeName=spine_04_capsule}]
 spine_05 [{shapeType=Capsule, length=10.5, radius=9, shapeName=spine_05_capsule, center=<-0.9828 1.3693 0>}]
 upperarm_l [{shapeType=Capsule, length=15, radius=4, shapeName=upperarm_l_capsule, center=<8 -0.8075 0>, rotation=<rot 86.5473 -0.003 -0.0031>}]
 upperarm_r [{shapeType=Capsule, length=14.5, radius=4.5, shapeName=upperarm_r_capsule, center=<-8 0 0>, rotation=<rot 90 0 0>}]
}
constraints [
 {bone1Name=dyn_hair_bk_03, bone2Name=dyn_hair_bk_02, swing1LimitDegrees=50, swing2LimitDegrees=15, twistLimitDegrees=5}
 {bone1Name=dyn_hair_bk_02, bone2Name=dyn_hair_bk_01, swing1LimitDegrees=50, swing2LimitDegrees=15, twistLimitDegrees=5}
 {bone1Name=dyn_hair_bk_01, bone2Name=dyn_hair_bk_00, swing1LimitDegrees=50, swing2LimitDegrees=15, twistLimitDegrees=5}
 {bone1Name=dyn_hair_bk_00, bone2Name=dyn_hair_bk_MSTR, swing1LimitDegrees=50, swing2LimitDegrees=15, twistLimitDegrees=5}
 {bone1Name=dyn_hair_l_02, bone2Name=dyn_hair_l_01, swing1LimitDegrees=45, swing2LimitDegrees=10, twistLimitDegrees=5}
 {bone1Name=dyn_hair_l_01, bone2Name=dyn_hair_l_00, swing1LimitDegrees=45, swing2LimitDegrees=10, twistLimitDegrees=5}
 {bone1Name=dyn_hair_l_00, bone2Name=dyn_hair_l_MSTR, swing1LimitDegrees=45, swing2LimitDegrees=10, twistLimitDegrees=5}
 {bone1Name=dyn_hair_r_02, bone2Name=dyn_hair_r_01, swing1LimitDegrees=45, swing2LimitDegrees=10, twistLimitDegrees=5}
 {bone1Name=dyn_hair_r_01, bone2Name=dyn_hair_r_00, swing1LimitDegrees=45, swing2LimitDegrees=10, twistLimitDegrees=5}
 {bone1Name=dyn_hair_r_00, bone2Name=dyn_hair_r_MSTR, swing1LimitDegrees=45, swing2LimitDegrees=10, twistLimitDegrees=5}
 {bone1Name=dyn_hair_r_03, bone2Name=dyn_hair_r_00, swing1LimitDegrees=20, swing2LimitDegrees=5, twistLimitDegrees=5}
]
