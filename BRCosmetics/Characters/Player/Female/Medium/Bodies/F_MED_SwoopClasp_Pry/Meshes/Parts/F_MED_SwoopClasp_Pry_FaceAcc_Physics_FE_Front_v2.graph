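# Physics asset for the left front dynamic hair strand on the female medium SwoopClasp_Pry character's face accessory mesh.
# Capsule bodies form a six-bone chain from a master bone through segments 00 to 04, with two capsules on the tip.
# Constraints link neighbouring segments with a wider first swing limit and tight second swing and twist limits.
# Separate capsules on the upper spine, left clavicle and left upper arm have no constraints, apparently serving as collision shapes.
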PhysicsAsset /BRCosmetics/Characters/Player/Female/Medium/Bodies/F_MED_SwoopClasp_Pry/Meshes/Parts/F_MED_SwoopClasp_Pry_FaceAcc_Physics_FE_Front_v2
bodies {
 clavicle_l [{shapeType=Capsule, length=4, radius=5.005, shapeName=clavicle_l_capsule, center=<8.003 -0.7579 -1.3847>, rotation=<rot 90 0 0>}]
 dyn_hair_l_Front_00 [{shapeType=Capsule, length=3, radius=2, shapeName=dyn_hair_l_Front_00_capsule, center=<2.4794 0 0>, rotation=<rot 90 0 0>}]
 dyn_hair_l_Front_01 [{shapeType=Capsule, length=3, radius=2, shapeName=dyn_hair_l_Front_01_capsule, center=<2.1329 0.8513 0>, rotation=<rot 90 22.1444 0>}]
 dyn_hair_l_Front_02 [{shapeType=Capsule, length=2.5, radius=2.5, shapeName=dyn_hair_l_Front_02_capsule, center=<2.7188 -0.7712 0>, rotation=<rot 90 22.7995 0>}]
 dyn_hair_l_Front_03 [{shapeType=Capsule, length=2.5, radius=2.5, shapeName=dyn_hair_l_Front_03_capsule, center=<2.4794 0 0>, rotation=<rot 90 0 0>}]
 dyn_hair_l_Front_04 [
  {
   shapeType=Capsule
   length=7
   radius=2
   shapeName=dyn_hair_l_Front_04_capsule
   center=<4.4329 -0.0819 0.3919>
   rotation=<rot 72.3916 -176.6763 -159.2147>
  }
  {
   shapeType=Capsule
   length=6.5
   radius=1.5
   shapeName=dyn_hair_l_Front_04_capsule
   center=<0.0009 -1.1448 1.1687>
   rotation=<rot -0.0256 0.6306 -33.9136>
  }
 ]
 dyn_hair_l_Front_MSTR [{shapeType=Capsule, length=1, radius=0.5, shapeName=dyn_hair_l_Front_MSTR_capsule}]
 spine_04 [{shapeType=Capsule, length=8, radius=7.005, shapeName=spine_04_capsule, center=<1.8303 4.6252 0>, rotation=<rot 0 -90 0>}]
 spine_05 [{shapeType=Capsule, length=20.5, radius=7, shapeName=spine_05_capsule, rotation=<rot 0 -90 0>}]
 upperarm_l [{
   shapeType=Capsule
   length=12.505
   radius=4.005
   shapeName=upperarm_l_capsule
   center=<8.4353 0 -0.9646>
   rotation=<rot 86.7793 -0.0018 -0.0019>
  }]
}
constraints [
 {bone1Name=dyn_hair_l_Front_04, bone2Name=dyn_hair_l_Front_03, swing1LimitDegrees=45, swing2LimitDegrees=5, twistLimitDegrees=5}
 {bone1Name=dyn_hair_l_Front_03, bone2Name=dyn_hair_l_Front_02, swing1LimitDegrees=45, swing2LimitDegrees=15, twistLimitDegrees=5}
 {bone1Name=dyn_hair_l_Front_02, bone2Name=dyn_hair_l_Front_01, swing1LimitDegrees=45, swing2LimitDegrees=5, twistLimitDegrees=5}
 {bone1Name=dyn_hair_l_Front_01, bone2Name=dyn_hair_l_Front_00, swing1LimitDegrees=45, swing2LimitDegrees=5, twistLimitDegrees=5}
 {bone1Name=dyn_hair_l_Front_00, bone2Name=dyn_hair_l_Front_MSTR, swing1LimitDegrees=45, swing2LimitDegrees=5, twistLimitDegrees=5}
]
