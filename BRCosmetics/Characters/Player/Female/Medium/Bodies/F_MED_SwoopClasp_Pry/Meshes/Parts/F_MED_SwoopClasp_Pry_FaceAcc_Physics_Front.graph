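# Physics asset for the left front dynamic hair chain on the female medium SwoopClasp_Pry character's face accessory.
# Capsules on the upper spine, neck, clavicles and upper arms form the body volumes, linked by constraints with free swing and twist.
# The hair chain hangs from neck_01 through a free-rotating master bone.
# Its four segments bend mainly on one swing axis, with tight second swing and twist limits and slightly looser second swing at the tip.
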PhysicsAsset /BRCosmetics/Characters/Player/Female/Medium/Bodies/F_MED_SwoopClasp_Pry/Meshes/Parts/F_MED_SwoopClasp_Pry_FaceAcc_Physics_Front
bodies {
 clavicle_l [{shapeType=Capsule, length=1.9169, radius=6, center=<6.4709 1.1493 -2.1395>, rotation=<rot 85.606 0 0>}]
 clavicle_r [{shapeType=Capsule, length=4.6669, radius=8.0644, center=<-7.6181 -0.4649 2.3477>, rotation=<rot 85.606 0 0>}]
 dyn_hair_l_Front_00 [{shapeType=Capsule, length=1.875, radius=3.125, shapeName=dyn_hair_r_Front_01_capsule, center=<3 0 0>, rotation=<rot 90 0 0>}]
 dyn_hair_l_Front_01 [{shapeType=Capsule, length=2.0625, radius=2, shapeName=dyn_hair_r_Front_02_capsule, center=<2.6486 1.8546 0>, rotation=<rot 90 35.0003 0>}]
 dyn_hair_l_Front_02 [{shapeType=Capsule, length=0, radius=2, shapeName=dyn_hair_r_Front_03_capsule, center=<3.816 1.4395 0>, rotation=<rot 90 -179.5524 175.4474>}]
 dyn_hair_l_Front_03 [{shapeType=Capsule, length=8, radius=2, shapeName=dyn_hair_r_Front_04_capsule, center=<6.3246 0.3397 0>, rotation=<rot 90 0 0>}]
 dyn_hair_l_Front_MSTR [{shapeType=Capsule, length=0.505, radius=0.505, shapeName=dyn_hair_r_Front_MSTR_capsule, rotation=<rot 0 -90 0>}]
 neck_01 [{shapeType=Capsule, length=4.255, radius=5.5675, shapeName=neck_01_capsule, rotation=<rot 0 -90 0>}]
 spine_04 [{shapeType=Capsule, length=4.5941, radius=10.0751, shapeName=spine_04_capsule, center=<2.5892 0.4958 0>, rotation=<rot 0 -92.6847 0>}]
 spine_05 [{shapeType=Capsule, length=15.7816, radius=9.5, shapeName=spine_04_capsule, center=<0.0821 0.0358 0>, rotation=<rot 0 -92.6847 0>}]
 upperarm_l [{shapeType=Capsule, length=13.4229, radius=5, shapeName=upperarm_r_capsule, center=<5.2485 1.0639 -0.2913>, rotation=<rot 0 -90 90>}]
 upperarm_r [{shapeType=Capsule, length=20.4229, radius=4.1429, shapeName=upperarm_r_capsule, center=<-11.013 1.25 -1.2498>, rotation=<rot 0 -90 90>}]
}
constraints [
 {
  bone1Name=spine_05
  bone2Name=spine_04
  swing1LimitDegrees=45
  swing1Motion=Free
  swing2LimitDegrees=45
  swing2Motion=Free
  twistLimitDegrees=45
  twistMotion=Free
 }
 {
  bone1Name=clavicle_l
  bone2Name=spine_05
  swing1LimitDegrees=45
  swing1Motion=Free
  swing2LimitDegrees=45
  swing2Motion=Free
  twistLimitDegrees=45
  twistMotion=Free
 }
 {
  bone1Name=upperarm_l
  bone2Name=clavicle_l
  swing1LimitDegrees=45
  swing1Motion=Free
  swing2LimitDegrees=45
  swing2Motion=Free
  twistLimitDegrees=45
  twistMotion=Free
 }
 {
  bone1Name=clavicle_r
  bone2Name=spine_05
  swing1LimitDegrees=45
  swing1Motion=Free
  swing2LimitDegrees=45
  swing2Motion=Free
  twistLimitDegrees=45
  twistMotion=Free
 }
 {
  bone1Name=upperarm_r
  bone2Name=clavicle_r
  swing1LimitDegrees=45
  swing1Motion=Free
  swing2LimitDegrees=45
  swing2Motion=Free
  twistLimitDegrees=45
  twistMotion=Free
 }
 {
  bone1Name=neck_01
  bone2Name=spine_05
  swing1LimitDegrees=45
  swing1Motion=Free
  swing2LimitDegrees=45
  swing2Motion=Free
  twistLimitDegrees=45
  twistMotion=Free
 }
 {
  bone1Name=dyn_hair_l_Front_MSTR
  bone2Name=neck_01
  swing1LimitDegrees=45
  swing1Motion=Free
  swing2LimitDegrees=45
  swing2Motion=Free
  twistLimitDegrees=45
  twistMotion=Free
 }
 {bone1Name=dyn_hair_l_Front_00, bone2Name=dyn_hair_l_Front_MSTR, swing1LimitDegrees=45, swing2LimitDegrees=5, twistLimitDegrees=5}
 {bone1Name=dyn_hair_l_Front_01, bone2Name=dyn_hair_l_Front_00, swing1LimitDegrees=45, swing2LimitDegrees=5, twistLimitDegrees=5}
 {bone1Name=dyn_hair_l_Front_02, bone2Name=dyn_hair_l_Front_01, swing1LimitDegrees=45, swing2LimitDegrees=5, twistLimitDegrees=5}
 {bone1Name=dyn_hair_l_Front_03, bone2Name=dyn_hair_l_Front_02, swing1LimitDegrees=45, swing2LimitDegrees=15, twistLimitDegrees=5}
]
