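# Animation Blueprint driving hair physics for the SwoopClasp_Pry female character's FaceAcc part.
# The copied parent pose gets bone constraints, crouch Control Rig, skydive/parachute offsets and backpack/physics-off fallback constraints, then is cached.
# Separate front-end and gameplay RigidBody chains simulate back and front hair, blend by bIsInFrontEnd, then feed AnimDynamics bangs and a final constraint.
# Update logic sets gravity and gating flags from movement, backbling, skydiving and rigid-body state.
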AnimBlueprint /BRCosmetics/Characters/Player/Female/Medium/Bodies/F_MED_SwoopClasp_Pry/Meshes/Parts/F_MED_SwoopClasp_Pry_FaceAcc_AnimBP
target_skeleton=&/F_MED_SwoopClasp_Pry_FaceAcc_Skeleton
graphs [
 AnimGraph {nodes [
   RigidBody!(
    ComponentPose
    bAlphaBoolEnabled:bool=Get BackpackOff() -> (BackpackOff:bool)
    OverrideWorldGravity:Vector=Get GravityOverride()
    ExternalForce
    ComponentLinearAccScale:Vector="0.3,0.3,0.35"
    ComponentLinearVelScale:Vector="0.3,0.3,0.35"
   ) -> (Pose) {
    alphaInputType=Bool
    bClampLinearTranslationLimitToRefPose=true
    bOverrideWorldGravity=true
    overridePhysicsAsset=&/F_MED_SwoopClasp_Pry_FaceAcc_Physics_Back
    simulationSpace=BaseBoneSpace
    alphaBoolBlend {blendOption=Linear}
    baseBoneRef {boneName=spine_05}
    componentAppliedLinearAccClamp=<1000 2500 2500>
    componentLinearAccScale=<0.3 0.3 0.5>
    componentLinearVelScale=<0.3 0.3 0.5>
    overrideWorldGravity {z=-800}
    simSpaceSettings {
     maxAngularAcceleration=1000
     maxAngularVelocity=3
     maxLinearAcceleration=700
     maxLinearVelocity=500
     velocityScaleZ=2
     worldAlpha=0.5
     externalLinearDragV=<0.2 0.2>
    }
   }
   n1 = RigidBody(
    ComponentPose=Local To Component(LocalPose=Use cached pose 'FrontEndPose'())
    bAlphaBoolEnabled=Get BackpackOff()
    OverrideWorldGravity=Get GravityOverride()
    ComponentLinearAccScale="0.2,0.2,0.2"
    ComponentLinearVelScale="0.2,0.2,0.2"
   ) {
    alphaInputType=Bool
    bClampLinearTranslationLimitToRefPose=true
    bOverrideWorldGravity=true
    overridePhysicsAsset=&F_MED_SwoopClasp_Pry_FaceAcc_Physics_Back
    simulationSpace=BaseBoneSpace
    alphaBoolBlend {blendOption=Linear}
    baseBoneRef {boneName=spine_05}
    componentAppliedLinearAccClamp=<1000 2500 2500>
    componentLinearAccScale=<0.3 0.3 0.5>
    componentLinearVelScale=<0.3 0.3 0.5>
    overrideWorldGravity {z=-800}
    simSpaceSettings {
     maxAngularAcceleration=1000
     maxAngularVelocity=3
     maxLinearAcceleration=700
     maxLinearVelocity=500
     velocityScaleZ=2
     worldAlpha=0.5
     externalLinearDragV=<0.2 0.2>
    }
   }
   n2 = RigidBody(
    ComponentPose=@n1
    bAlphaBoolEnabled=Get Not in SkyDiving() -> (Not in SkyDiving:bool)
    OverrideWorldGravity=Get GravityOverride()
    ComponentLinearAccScale="0.3,0.3,0.35"
    ComponentLinearVelScale="0.3,0.3,0.35"
   ) {
    alphaInputType=Bool
    bClampLinearTranslationLimitToRefPose=true
    bOverrideWorldGravity=true
    overridePhysicsAsset=&/F_MED_SwoopClasp_Pry_FaceAcc_Physics_Front
    simulationSpace=BaseBoneSpace
    alphaBoolBlend {blendOption=Linear}
    baseBoneRef {boneName=spine_05}
    componentAppliedLinearAccClamp=<1000 2500 2500>
    componentLinearAccScale=<0.3 0.3 0.5>
    componentLinearVelScale=<0.3 0.3 0.5>
    overrideWorldGravity {z=-800}
    simSpaceSettings {
     maxAngularAcceleration=1000
     maxAngularVelocity=3
     maxLinearAcceleration=700
     maxLinearVelocity=500
     velocityScaleZ=2
     worldAlpha=0.5
     externalLinearDragV=<0.2 0.2>
    }
   }
   n3 = RigidBody(
    ComponentPose=@n2
    bAlphaBoolEnabled=Get bIsSkyDiving()
    OverrideWorldGravity="0,0,-980"
    ComponentLinearAccScale="0,0,0"
    ComponentLinearVelScale="0,0,0"
   ) {
    alphaInputType=Bool
    bClampLinearTranslationLimitToRefPose=true
    bOverrideWorldGravity=true
    overridePhysicsAsset=&F_MED_SwoopClasp_Pry_FaceAcc_Physics_Front
    simulationSpace=BaseBoneSpace
    alphaBoolBlend {blendOption=Linear}
    baseBoneRef {boneName=spine_05}
    componentAppliedLinearAccClamp=<100 100 100>
    componentLinearAccScale=<0.3 0.3 0.5>
    componentLinearVelScale=<0.3 0.3 0.5>
    overrideWorldGravity {z=-800}
    simSpaceSettings {
     maxAngularAcceleration=1000
     maxAngularVelocity=3
     maxLinearAcceleration=700
     maxLinearVelocity=500
     velocityScaleZ=0.3
     worldAlpha=0.5
     externalLinearDragV=<0.2 0.2>
    }
   }
   n4 = Component To Local(ComponentPose=@n3)
   n5 = RigidBody(
    ComponentPose=Local To Component(LocalPose=Use cached pose 'FrontEndPose'())
    bAlphaBoolEnabled=Get BackpackOff()
    OverrideWorldGravity=Get GravityOverride()
    ComponentLinearAccScale="0.2,0.2,0.2"
    ComponentLinearVelScale="0.2,0.2,0.2"
   ) {
    alphaInputType=Bool
    bClampLinearTranslationLimitToRefPose=true
    bOverrideWorldGravity=true
    overridePhysicsAsset=&/F_MED_SwoopClasp_Pry_FaceAcc_Physics_FE_Back_v2
    simulationSpace=BaseBoneSpace
    alphaBoolBlend {blendOption=Linear}
    baseBoneRef {boneName=spine_05}
    componentAppliedLinearAccClamp=<1000 2500 2500>
    componentLinearAccScale=<0.3 0.3 0.5>
    componentLinearVelScale=<0.3 0.3 0.5>
    overrideWorldGravity {z=-800}
    simSpaceSettings {
     maxAngularAcceleration=1000
     maxAngularVelocity=3
     maxLinearAcceleration=700
     maxLinearVelocity=500
     velocityScaleZ=2
     worldAlpha=0.5
     externalLinearDragV=<0.2 0.2>
    }
   }
   n6 = RigidBody(
    ComponentPose=@n5
    OverrideWorldGravity=Get GravityOverride()
    ComponentLinearAccScale="0.3,0.3,0.35"
    ComponentLinearVelScale="0.3,0.3,0.35"
   ) {
    alphaInputType=Bool
    bClampLinearTranslationLimitToRefPose=true
    bOverrideWorldGravity=true
    overridePhysicsAsset=&/F_MED_SwoopClasp_Pry_FaceAcc_Physics_FE_Front_v2
    simulationSpace=BaseBoneSpace
    alphaBoolBlend {blendOption=Linear}
    baseBoneRef {boneName=spine_05}
    componentAppliedLinearAccClamp=<1000 2500 2500>
    componentLinearAccScale=<0.3 0.3 0.5>
    componentLinearVelScale=<0.3 0.3 0.5>
    overrideWorldGravity {z=-800}
    simSpaceSettings {
     maxAngularAcceleration=1000
     maxAngularVelocity=3
     maxLinearAcceleration=700
     maxLinearVelocity=500
     velocityScaleZ=2
     worldAlpha=0.5
     externalLinearDragV=<0.2 0.2>
    }
   }
   n7 = Component To Local(ComponentPose=@n6)
   n8 = Get bIsInFrontEnd()
   n9 = Anim Dynamics - Bone: dyn_hair_bangs_r_01!(
    ComponentPose=Local To Component(LocalPose=Blend Poses by bool(BlendPose_0=@n7, BlendPose_1=@n4, bActiveValue=@n8) {blendType=Linear})
    bAlphaBoolEnabled:bool=True
    ExternalForce="0,0,0"
   ) -> (Pose) {
    alphaInputType=Bool
    angularDampingOverride=0.9999
    bOverrideAngularDamping=true
    bUseGravityOverride=true
    simulationSpace=RootRelative
    alphaBoolBlend {blendInTime=0.1, blendOption=Linear, blendOutTime=0.2}
    boundBone {boneName=dyn_hair_bangs_r_01}
    componentAppliedLinearAccClamp=<1500 1500 1500>
    componentLinearAccScale=<0.2 0.2 0.2>
    componentLinearVelScale=<0.3 0.3 0.3>
    gravityOverride {z=-980}
    physicsBodyDefinitions [{
      boundBone {boneName=dyn_hair_bangs_r_01}
      boxExtents=<10 5 5>
      constraintSetup {angularLimitsMax=<5 15 30>, angularLimitsMin=<-5 -5 0>, angularTarget=<0 0 0>}
      localJointOffset=<1 0 0>
     }]
   }
   n10 = Anim Dynamics - Bone: dyn_hair_bangs_l_01!(ComponentPose=@n9, bAlphaBoolEnabled:bool=True, ExternalForce="0,0,0") -> (Pose) {
    alphaInputType=Bool
    angularDampingOverride=0.999
    bOverrideAngularDamping=true
    bUseGravityOverride=true
    simulationSpace=RootRelative
    alphaBoolBlend {blendInTime=0.1, blendOption=Linear, blendOutTime=0.2}
    boundBone {boneName=dyn_hair_bangs_l_01}
    componentAppliedLinearAccClamp=<3000 3000 3000>
    componentLinearAccScale=<0.2 0.2 0.2>
    componentLinearVelScale=<0.3 0.3 0.3>
    gravityOverride {z=-980}
    physicsBodyDefinitions [{
      boundBone {boneName=dyn_hair_bangs_l_01}
      boxExtents=<14 8 8>
      constraintSetup {angularLimitsMax=<5 15 30>, angularLimitsMin=<-5 -5 -5>, angularTarget=<0 0 0>}
      localJointOffset=<1 0 0>
     }]
   }
   n11 = "Constraint - dyn_hair_r_03 (PPPPPPP)"!(
    ComponentPose=@n10
    Alpha=0.5
    ConstraintWeights_0=0.0063
    ConstraintWeights_1:float=0.0304
    ConstraintWeights_2:float=0.5485
    ConstraintWeights_3:float=0.2844
    ConstraintWeights_4:float=0.0968
    ConstraintWeights_5:float=0.0038
    ConstraintWeights_6:float=0.0298
   ) -> (Pose) {
    alphaBoolBlend {blendOption=Linear}
    boneToModify {boneName=dyn_hair_r_03}
    constraintSetup [
     {targetBone {boneName=spine_04}}
     {targetBone {boneName=spine_03}}
     {targetBone {boneName=spine_05}}
     {targetBone {boneName=clavicle_r}}
     {targetBone {boneName=neck_01}}
     {targetBone {boneName=head}}
     {targetBone {boneName=neck_02}}
    ]
    constraintWeights [1,1,1,1,1,1,1]
   }
   Output Pose(Result=Component To Local(ComponentPose=@n11)) {layerGroup=None}
   n12 = "Constraint - dyn_hair_l_MSTR (PP)"!(
    ComponentPose=Local To Component(LocalPose=Copy Pose From Mesh() {bUseAttachedParent=true})
    Alpha=1
    ConstraintWeights_0=0.8
    ConstraintWeights_1:float=0.2
   ) -> (Pose) {
    alphaBoolBlend {blendOption=Linear}
    boneToModify {boneName=dyn_hair_l_MSTR}
    constraintSetup [{targetBone {boneName=spine_05}}, {targetBone {boneName=head}}]
    constraintWeights [1,1]
   }
   n13 = "Constraint - dyn_hair_r_MSTR (PP)"!(ComponentPose=@n12, Alpha=1, ConstraintWeights_0=0.8, ConstraintWeights_1:float=0.2) -> (Pose) {
    alphaBoolBlend {blendOption=Linear}
    boneToModify {boneName=dyn_hair_r_MSTR}
    constraintSetup [{targetBone {boneName=spine_05}}, {targetBone {boneName=head}}]
    constraintWeights [1,1]
   }
   n14 = "Constraint - dyn_hair_bk_MSTR (PP)"!(ComponentPose=@n13, Alpha=1, ConstraintWeights_0=0.8, ConstraintWeights_1:float=0.2) -> (Pose) {
    alphaBoolBlend {blendOption=Linear}
    boneToModify {boneName=dyn_hair_bk_MSTR}
    constraintSetup [{targetBone {boneName=spine_05}}, {targetBone {boneName=head}}]
    constraintWeights [1,1]
   }
   n15 = "Constraint - dyn_hair_l_Front_MSTR (PP)"!(ComponentPose=@n14, Alpha=1, ConstraintWeights_0=0.3, ConstraintWeights_1:float=0.7) -> (Pose) {
    alphaBoolBlend {blendOption=Linear}
    boneToModify {boneName=dyn_hair_l_Front_MSTR}
    constraintSetup [{targetBone {boneName=clavicle_l}}, {targetBone {boneName=head}}]
    constraintWeights [1,1]
   }
   n16 = Control Rig!(
    Source=Component To Local(ComponentPose=@n15)
    Alpha
    crouch move:double=Get crouchmove() -> (crouchmove:double)
    crouch:double=Get crouch() -> (crouch:double)
   ) -> (Pose) {
    alphaBoolBlend {blendOption=Linear}
    controlRigAssetReference {blueprintRigClass=&/F_MED_SwoopClasp_Pry_FaceAcc_CtrlRig.F_MED_SwoopClasp_Pry_FaceAcc_CtrlRig_C}
   }
   n17 = Get bIsSkyDiving()
   n18 = "Transform (Modify) Bone - Bone: dyn_hair_bk_MSTR"!(ComponentPose=Local To Component(LocalPose=@n16), bAlphaBoolEnabled:bool=@n17, Translation="2,-2,0", Rotation="0,25,0", Scale="1,1,1") -> (Pose) {
    alphaInputType=Bool
    rotationMode=BMM_Additive
    rotationSpace=BCS_BoneSpace
    translationMode=BMM_Additive
    translationSpace=BCS_BoneSpace
    alphaBoolBlend {blendInTime=0.5, blendOption=Linear, blendOutTime=0.3}
    alphaScaleBiasClamp {bInterpResult=true, interpSpeedIncreasing=2}
    boneToModify {boneName=dyn_hair_bk_MSTR}
    rotation {yaw=25}
    translation=<2 -2>
   }
   n19 = "Transform (Modify) Bone - Bone: dyn_hair_r_MSTR"!(ComponentPose=@n18, bAlphaBoolEnabled:bool=@n17, Translation="2,-2,0", Rotation="0,25,0", Scale="1,1,1") -> (Pose) {
    alphaInputType=Bool
    rotationMode=BMM_Additive
    rotationSpace=BCS_BoneSpace
    translationMode=BMM_Additive
    translationSpace=BCS_BoneSpace
    alphaBoolBlend {blendInTime=0.5, blendOption=Linear, blendOutTime=0.3}
    alphaScaleBiasClamp {bInterpResult=true, interpSpeedIncreasing=2}
    boneToModify {boneName=dyn_hair_r_MSTR}
    rotation {yaw=25}
    translation=<2 -2>
   }
   n20 = "Transform (Modify) Bone - Bone: dyn_hair_l_MSTR"!(ComponentPose=@n19, bAlphaBoolEnabled:bool=@n17, Translation="2,-2,0", Rotation="0,0,0", Scale="1,1,1") -> (Pose) {
    alphaInputType=Bool
    rotationMode=BMM_Additive
    rotationSpace=BCS_BoneSpace
    translationMode=BMM_Additive
    translationSpace=BCS_BoneSpace
    alphaBoolBlend {blendInTime=0.5, blendOption=Linear, blendOutTime=0.3}
    alphaScaleBiasClamp {bInterpResult=true, interpSpeedIncreasing=2}
    boneToModify {boneName=dyn_hair_l_MSTR}
    translation=<2 -2>
   }
   n21 = Get bIsParachuteOpened()
   n22 = "Transform (Modify) Bone - Bone: dyn_hair_bk_MSTR"(ComponentPose=@n20, bAlphaBoolEnabled=@n21, Translation="-2,0,0", Rotation="0,0,0", Scale="1,1,1") {
    alphaInputType=Bool
    rotationMode=BMM_Additive
    rotationSpace=BCS_BoneSpace
    translationMode=BMM_Additive
    translationSpace=BCS_BoneSpace
    alphaBoolBlend {blendInTime=0.1, blendOption=Linear, blendOutTime=0.3}
    alphaScaleBiasClamp {bInterpResult=true, interpSpeedIncreasing=2}
    boneToModify {boneName=dyn_hair_bk_MSTR}
    translation {x=-2}
   }
   n23 = "Transform (Modify) Bone - Bone: dyn_hair_r_MSTR"(ComponentPose=@n22, bAlphaBoolEnabled=@n21, Translation="-2,0,-2", Rotation="0,0,0", Scale="1,1,1") {
    alphaInputType=Bool
    rotationMode=BMM_Additive
    rotationSpace=BCS_BoneSpace
    translationMode=BMM_Additive
    translationSpace=BCS_BoneSpace
    alphaBoolBlend {blendInTime=0.1, blendOption=Linear, blendOutTime=0.3}
    alphaScaleBiasClamp {bInterpResult=true, interpSpeedIncreasing=2}
    boneToModify {boneName=dyn_hair_r_MSTR}
    translation {x=-2, z=-2}
   }
   n24 = "Transform (Modify) Bone - Bone: dyn_hair_l_MSTR"(ComponentPose=@n23, bAlphaBoolEnabled=@n21, Translation="-2,0,2", Rotation="0,0,0", Scale="1,1,1") {
    alphaInputType=Bool
    rotationMode=BMM_Additive
    rotationSpace=BCS_BoneSpace
    translationMode=BMM_Additive
    translationSpace=BCS_BoneSpace
    alphaBoolBlend {blendInTime=0.1, blendOption=Linear, blendOutTime=0.3}
    alphaScaleBiasClamp {bInterpResult=true, interpSpeedIncreasing=2}
    boneToModify {boneName=dyn_hair_l_MSTR}
    translation {x=-2, z=2}
   }
   n25 = "Transform (Modify) Bone - Bone: dyn_hair_l_Front_MSTR"!(ComponentPose=@n24, bAlphaBoolEnabled:bool=@n21, Translation="-2,0,0", Rotation="0,5,0", Scale="1,1,1") -> (Pose) {
    alphaInputType=Bool
    rotationMode=BMM_Additive
    rotationSpace=BCS_BoneSpace
    translationMode=BMM_Additive
    translationSpace=BCS_BoneSpace
    alphaBoolBlend {blendInTime=0.1, blendOption=Linear, blendOutTime=0.3}
    alphaScaleBiasClamp {bInterpResult=true, interpSpeedIncreasing=2}
    boneToModify {boneName=dyn_hair_l_Front_MSTR}
    rotation {yaw=5}
    translation {x=-2}
   }
   n26 = Get BackpackPerf() -> (BackpackPerf:bool)
   n27 = "Constraint - dyn_hair_bk_01 (PP)"!(ComponentPose=@n25, bAlphaBoolEnabled:bool=@n26, ConstraintWeights_0=0.2, ConstraintWeights_1:float=0.8) -> (Pose) {
    alphaInputType=Bool
    alphaBoolBlend {blendInTime=0.1, blendOption=Linear, blendOutTime=0.1}
    boneToModify {boneName=dyn_hair_bk_01}
    constraintSetup [{targetBone {boneName=head}}, {targetBone {boneName=spine_05}}]
    constraintWeights [1,1]
   }
   n28 = "Constraint - dyn_hair_bk_02 (PP)"!(ComponentPose=@n27, bAlphaBoolEnabled:bool=@n26, ConstraintWeights_0=0.7, ConstraintWeights_1:float=0.3) -> (Pose) {
    alphaInputType=Bool
    alphaBoolBlend {blendInTime=0.1, blendOption=Linear, blendOutTime=0.1}
    boneToModify {boneName=dyn_hair_bk_02}
    constraintSetup [{targetBone {boneName=spine_05}}, {targetBone {boneName=spine_04}}]
    constraintWeights [1,1]
   }
   n29 = "Constraint - dyn_hair_l_01 (PP)"!(ComponentPose=@n28, bAlphaBoolEnabled:bool=@n26, ConstraintWeights_0=0.8, ConstraintWeights_1:float=0.2) -> (Pose) {
    alphaInputType=Bool
    alphaBoolBlend {blendInTime=0.1, blendOption=Linear, blendOutTime=0.1}
    boneToModify {boneName=dyn_hair_l_01}
    constraintSetup [{targetBone {boneName=spine_05}}, {targetBone {boneName=head}}]
    constraintWeights [1,1]
   }
   n30 = "Constraint - dyn_hair_l_02 (PP)"!(ComponentPose=@n29, bAlphaBoolEnabled:bool=@n26, ConstraintWeights_0=0.3, ConstraintWeights_1:float=0.7) -> (Pose) {
    alphaInputType=Bool
    alphaBoolBlend {blendInTime=0.1, blendOption=Linear, blendOutTime=0.1}
    boneToModify {boneName=dyn_hair_l_02}
    constraintSetup [{targetBone {boneName=spine_04}}, {targetBone {boneName=spine_05}}]
    constraintWeights [1,1]
   }
   n31 = "Constraint - dyn_hair_r_01 (PP)"!(ComponentPose=@n30, bAlphaBoolEnabled:bool=@n26, ConstraintWeights_0=0.8, ConstraintWeights_1:float=0.2) -> (Pose) {
    alphaInputType=Bool
    alphaBoolBlend {blendInTime=0.1, blendOption=Linear, blendOutTime=0.1}
    boneToModify {boneName=dyn_hair_r_01}
    constraintSetup [{targetBone {boneName=spine_05}}, {targetBone {boneName=head}}]
    constraintWeights [1,1]
   }
   n32 = "Constraint - dyn_hair_r_02 (PP)"!(ComponentPose=@n31, bAlphaBoolEnabled:bool=@n26, ConstraintWeights_0=0.3, ConstraintWeights_1:float=0.7) -> (Pose) {
    alphaInputType=Bool
    alphaBoolBlend {blendInTime=0.1, blendOption=Linear, blendOutTime=0.1}
    boneToModify {boneName=dyn_hair_r_02}
    constraintSetup [{targetBone {boneName=spine_04}}, {targetBone {boneName=spine_05}}]
    constraintWeights [1,1]
   }
   n33 = "Constraint - dyn_hair_r_03 (PP)"!(ComponentPose=@n32, bAlphaBoolEnabled:bool=@n26, ConstraintWeights_0=0.9, ConstraintWeights_1:float=0.1) -> (Pose) {
    alphaInputType=Bool
    alphaBoolBlend {blendInTime=0.1, blendOption=Linear, blendOutTime=0.1}
    boneToModify {boneName=dyn_hair_r_03}
    constraintSetup [{targetBone {boneName=spine_05}}, {targetBone {boneName=head}}]
    constraintWeights [1,1]
   }
   n34 = Get RigidBodyOff() -> (RigidBodyOff:bool)
   n35 = "Constraint - dyn_hair_l_Front_02 (PP)"!(ComponentPose=@n33, bAlphaBoolEnabled:bool=@n34, ConstraintWeights_0=0.3, ConstraintWeights_1:float=0.7) -> (Pose) {
    alphaInputType=Bool
    alphaBoolBlend {blendInTime=0.1, blendOption=Linear, blendOutTime=0.1}
    boneToModify {boneName=dyn_hair_l_Front_02}
    constraintSetup [{targetBone {boneName=spine_04}}, {targetBone {boneName=spine_05}}]
    constraintWeights [1,1]
   }
   n36 = "Constraint - dyn_hair_l_Front_03 (PP)"!(ComponentPose=@n35, bAlphaBoolEnabled:bool=@n34, ConstraintWeights_0=0.3, ConstraintWeights_1:float=0.7) -> (Pose) {
    alphaInputType=Bool
    alphaBoolBlend {blendInTime=0.1, blendOption=Linear, blendOutTime=0.1}
    boneToModify {boneName=dyn_hair_l_Front_03}
    constraintSetup [{targetBone {boneName=spine_04}}, {targetBone {boneName=spine_05}}]
    constraintWeights [1,1]
   }
   n37 = "Constraint - dyn_hair_l_Front_04 (PP)"!(ComponentPose=@n36, bAlphaBoolEnabled:bool=@n34, ConstraintWeights_0=0.3, ConstraintWeights_1:float=0.7) -> (Pose) {
    alphaInputType=Bool
    alphaBoolBlend {blendInTime=0.1, blendOption=Linear, blendOutTime=0.1}
    boneToModify {boneName=dyn_hair_l_Front_04}
    constraintSetup [{targetBone {boneName=spine_04}}, {targetBone {boneName=spine_05}}]
    constraintWeights [1,1]
   }
   Save cached pose 'FrontEndPose'(Pose=Component To Local(ComponentPose=@n37)) {cacheName=FrontEndPose}
  ]}
 EventGraph {nodes [
   Event BlueprintUpdateAnimation()
   > n1 = Sequence!(execute) -> (then_0, then_1, then_2:Exec, then_3:Exec, then_4:Exec)
   > n2 = Gravity Override State Function DEPRECATED(
    CharacterPartAnimInstance=Self-Reference()
    GameTimeInSeconds=GetGameTimeInSeconds()
    GravityOverrideParams=Get GravityOverrideParams() -> (GravityOverrideParams:Gravity Override Params Struct Structure)
   )
   > Set GravityOverride(GravityOverride=@n2)
   > n3 = Branch(Condition=CheckForBackbling())
   > Set ACC(ACC:Vector="0, 0, 0") -> (Output_Get:Vector)
   > Set VEL(VEL:Vector="0, 0, 0") -> (Output_Get:Vector)
   > Set GravityOverride(GravityOverride="0,0,-980")
   Set ACC(execute=@n3.else, ACC="0.2,0.3,0.3")
   > Set VEL(VEL="0.2,0.3,0.3")
   n4 = CheckForBackbling()
   Set Is backpack on(
    execute=@n1.then_1
    Is backpack on:bool=OR Boolean!(A=@n4, B=@n4.bIsBackblingACape, C:bool=@n4.bIsBackblingLong, D:bool=Get bIsDBNOCarried()) -> (ReturnValue)
   ) -> (Output_Get:bool)
   n5 = Get bRigidBodyEnabled()
   > Set BackpackPerf(BackpackPerf:bool=OR Boolean!(A=Get Is backpack on() -> (Is backpack on:bool), B=NOT Boolean(A=@n5)) -> (ReturnValue)) -> (Output_Get:bool)
   > Set RigidBodyOff(RigidBodyOff:bool=NOT Boolean(A=Get bRigidBodyEnabled())) -> (Output_Get:bool)
   > Set BackpackOff(BackpackOff:bool=NOT Boolean(A=Get Is backpack on())) -> (Output_Get:bool)
   > Set Not in SkyDiving(Not in SkyDiving:bool=NOT Boolean(A=Get bIsSkyDiving())) -> (Output_Get:bool)
   n6 = Get bIsCrouching()
   Set crouch(execute=@n1.then_3, crouch:double="To Float (Boolean)"(InBool=@n6)) -> (Output_Get:double)
   n7 = Branch(execute=@n1.then_2, Condition=AND Boolean(A=@n6, B=Get bIsMoving2D()))
   > Set crouchmove(crouchmove:double=1) -> (Output_Get:double)
   Set crouchmove(execute=@n7.else, crouchmove=0)
  ]}
]
variables [
 ACC {type=Vector}
 backpack {type=number}
 BackpackOff {type=boolean}
 BackpackPerf {type=boolean}
 crouch {type=number}
 crouchmove {type=number}
 Gravity Override Params {
  type=GravityOverrideParamsStruct
  default {
   onGround_Falling {gravityOverride=<0 0 -1250>}
   onGround_Moving {gravityOverride=<0 0 -1150>}
   skydive_FloatingDown {gravityOverride=<0 -100 50>, windAmplitude=<75 200 150>, windFrequency=<1 3.5 6>, windOffset=<0 -200 125>}
   skydive_GoingUp {gravityOverride=<0 -300 -600>}
   tactical_Sprint {gravityOverride=<0 0 -1150>}
  }
 }
 GravityOverride {type=Vector}
 GravityOverrideParams {
  type=GravityOverrideParamsStruct
  default {
   onGround_Falling {gravityOverride=<0 0 -1250>}
   onGround_Moving {gravityOverride=<0 0 -1150>}
   skydive_DivingDown {windFrequency=<1.5 1 1>}
   skydive_FloatingDown {gravityOverride=<0 -700 100>, windAmplitude=<100 -500 100>, windFrequency=<1.5 1 1>, windOffset=<0 -600 100>}
   skydive_GoingUp {gravityOverride=<0 -300 100>}
   tactical_Sprint {gravityOverride=<0 0 -1150>}
  }
 }
 Is backpack on {type=boolean}
 jumping {type=number}
 Not in SkyDiving {type=boolean}
 RBAN Params {
  type=RBANParamsStruct
  default {
   dBNO {simSpaceSettings {
     maxAngularAcceleration=1000
     maxAngularVelocity=3
     maxLinearAcceleration=300
     maxLinearVelocity=700
     velocityScaleZ=0
     externalLinearDragV=<0 0.5 0>
    }}
   dBNO_Carried {simSpaceSettings {
     maxAngularAcceleration=1000
     maxAngularVelocity=3
     maxLinearAcceleration=300
     maxLinearVelocity=700
     velocityScaleZ=0
     externalLinearDragV=<0 0.5 0>
    }}
   emoteOrMelee {simSpaceSettings {
     maxAngularAcceleration=1000
     maxAngularVelocity=3
     maxLinearAcceleration=300
     maxLinearVelocity=700
     velocityScaleZ=0
     externalLinearDragV=<0 0.5 0>
    }}
   front_End {simSpaceSettings {
     maxAngularAcceleration=1000
     maxAngularVelocity=3
     maxLinearAcceleration=300
     maxLinearVelocity=700
     velocityScaleZ=0
     externalLinearDragV=<0 0.5 0>
    }}
   onGround_Crouch_Moving {simSpaceSettings {
     maxAngularAcceleration=1000
     maxAngularVelocity=3
     maxLinearAcceleration=300
     maxLinearVelocity=700
     velocityScaleZ=0
     externalLinearDragV=<0 0.5 0>
    }}
   onGround_Falling {simSpaceSettings {
     maxAngularAcceleration=1000
     maxAngularVelocity=3
     maxLinearAcceleration=300
     maxLinearVelocity=700
     velocityScaleZ=0
     externalLinearDragV=<0 0.5 0>
    }}
   onGround_Moving {simSpaceSettings {
     maxAngularAcceleration=1000
     maxAngularVelocity=3
     maxLinearAcceleration=300
     maxLinearVelocity=700
     velocityScaleZ=0
     externalLinearDragV=<0 0.5 0>
    }}
   onGround_Standing {simSpaceSettings {
     maxAngularAcceleration=1000
     maxAngularVelocity=3
     maxLinearAcceleration=300
     maxLinearVelocity=700
     velocityScaleZ=0
     externalLinearDragV=<0 0.5 0>
    }}
   rail_Grinding {simSpaceSettings {
     maxAngularAcceleration=1000
     maxAngularVelocity=3
     maxLinearAcceleration=300
     maxLinearVelocity=700
     velocityScaleZ=0
     worldAlpha=0.1
     externalLinearDragV=<0 0.5 0>
    }}
   riding_Moving {simSpaceSettings {
     maxAngularAcceleration=1000
     maxAngularVelocity=3
     maxLinearAcceleration=300
     maxLinearVelocity=700
     velocityScaleZ=0
     externalLinearDragV=<0 0.5 0>
    }}
   riding_Moving_Targeting {simSpaceSettings {
     maxAngularAcceleration=1000
     maxAngularVelocity=3
     maxLinearAcceleration=300
     maxLinearVelocity=700
     velocityScaleZ=0
     externalLinearDragV=<0 0.5 0>
    }}
   riding_Standing {simSpaceSettings {
     maxAngularAcceleration=1000
     maxAngularVelocity=3
     maxLinearAcceleration=300
     maxLinearVelocity=700
     velocityScaleZ=0
     externalLinearDragV=<0 0.5 0>
    }}
   skydive_DivingDown {simSpaceSettings {
     maxAngularAcceleration=1000
     maxAngularVelocity=3
     maxLinearAcceleration=300
     maxLinearVelocity=700
     velocityScaleZ=0
     externalLinearDragV=<0 0.5 0>
    }}
   skydive_FloatingDown {
    componentLinearAccScale=<0 0.5 0>
    componentLinearVelScale=<0 0.5 0>
    simSpaceSettings {
     maxAngularAcceleration=1000
     maxAngularVelocity=3
     maxLinearAcceleration=300
     maxLinearVelocity=700
     velocityScaleZ=0
     externalLinearDragV=<0 0.5 0>
    }
   }
   skydive_GoingUp {simSpaceSettings {
     maxAngularAcceleration=1000
     maxAngularVelocity=3
     maxLinearAcceleration=300
     maxLinearVelocity=700
     velocityScaleZ=0
     externalLinearDragV=<0 0.5 0>
    }}
   skydive_Parachuting {
    componentLinearAccScale=<0 0.1 0.3>
    componentLinearVelScale=<0 0.1 0.3>
    simSpaceSettings {
     maxAngularAcceleration=1000
     maxAngularVelocity=3
     maxLinearAcceleration=300
     maxLinearVelocity=700
     velocityScaleZ=0
     externalLinearDragV=<0 0.5 0>
    }
   }
   skydive_Surfing {simSpaceSettings {
     maxAngularAcceleration=1000
     maxAngularVelocity=3
     maxLinearAcceleration=300
     maxLinearVelocity=700
     velocityScaleZ=0
     externalLinearDragV=<0 0.5 0>
    }}
   swimming {simSpaceSettings {
     maxAngularAcceleration=1000
     maxAngularVelocity=3
     maxLinearAcceleration=300
     maxLinearVelocity=700
     velocityScaleZ=0
     externalLinearDragV=<0 0.5 0>
    }}
   tactical_Sprint {simSpaceSettings {
     maxAngularAcceleration=1000
     maxAngularVelocity=3
     maxLinearAcceleration=300
     maxLinearVelocity=700
     velocityScaleZ=0
     externalLinearDragV=<0 0.5 0>
    }}
   ziplining {simSpaceSettings {
     maxAngularAcceleration=1000
     maxAngularVelocity=3
     maxLinearAcceleration=300
     maxLinearVelocity=700
     velocityScaleZ=0
     externalLinearDragV=<0 0.5 0>
    }}
  }
 }
 RBAN Settings Out {
  type=RBANSettingStruct
  default {
   componentLinearAccScale=<0 0 0>
   componentLinearVelScale=<0 0 0>
   simSpaceSettings {
    dampingAlpha=1
    maxAngularAcceleration=1000
    maxAngularVelocity=3
    maxLinearAcceleration=300
    maxLinearVelocity=700
    velocityScaleZ=0
    worldAlpha=1
    externalAngularVelocity=<0 0 0>
    externalLinearDragV=<0 0.5 0>
    externalLinearVelocity=<0 0 0>
   }
  }
 }
 RigidBodyOff {type=boolean}
 VEL {type=Vector}
]
